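# Physics asset for the right-leg dynamic bones of the medium male Racer Zero Jacket character body, covering the shoe tongue and heel.
# Capsule bodies form two short chains, each linking a dynamic bone to a small MSTR body that connects to the calf capsule.
# Tongue and heel joints have tight swing and twist limits, while both MSTR-to-calf constraints leave swing and twist free.
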PhysicsAsset /BRCosmetics/Characters/Player/Male/Medium/Bodies/M_MED_Racer_Zero_Jacket/Meshes/M_MED_Racer_Zero_Jacket_Dyn_leg_r
bodies {
 calf_r [{shapeType=Capsule, length=2.3856, radius=2.5778, center=<33.6406 -1.0186 0.0614>, rotation=<rot 90 -90 -92.1447>}]
 dyn_shoe_heel_r_01 [{shapeType=Capsule, length=3.125, radius=0.875, center=<1 0 0.0001>, rotation=<rot 90 0 0>}]
 dyn_shoe_heel_r_MSTR [{shapeType=Capsule, length=0.505, radius=0.505}]
 dyn_shoe_tongue_r [
  {shapeType=Capsule, length=3.9284, radius=1.0055, center=<4.4 0 0.0001>, rotation=<rot 90 0 0>}
  {shapeType=Capsule, length=1.7181, radius=1.0055, center=<2.4787 4.4931 -2.6644>, rotation=<rot 52.7479 -52.1248 -64.085>}
 ]
 dyn_shoe_tongue_r_MSTR [{shapeType=Capsule, length=0.505, radius=0.505}]
}
constraints [
 {bone1Name=dyn_shoe_tongue_r, bone2Name=dyn_shoe_tongue_r_MSTR, swing1LimitDegrees=20, swing2LimitDegrees=45, twistLimitDegrees=5}
 {bone1Name=dyn_shoe_heel_r_01, bone2Name=dyn_shoe_heel_r_MSTR, swing1LimitDegrees=20, swing2LimitDegrees=30, twistLimitDegrees=5}
 {
  bone1Name=dyn_shoe_tongue_r_MSTR
  bone2Name=calf_r
  swing1LimitDegrees=45
  swing1Motion=Free
  swing2LimitDegrees=45
  swing2Motion=Free
  twistLimitDegrees=45
  twistMotion=Free
 }
 {
  bone1Name=dyn_shoe_heel_r_MSTR
  bone2Name=calf_r
  swing1LimitDegrees=45
  swing1Motion=Free
  swing2LimitDegrees=45
  swing2Motion=Free
  twistLimitDegrees=45
  twistMotion=Free
 }
]
