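# Physics asset for the chest zipper pull on the male medium Racer Zero jacket cosmetic, set up for dynamic secondary motion.
# The small zipper capsule hangs from a master zipper body with tight twist and moderate swing limits.
# That master links to an upper-spine capsule, which links to the pelvis, both joints with free angular motion.
# The large spine and pelvis capsules approximate the torso around the zipper.
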PhysicsAsset /BRCosmetics/Characters/Player/Male/Medium/Bodies/M_MED_Racer_Zero_Jacket/Meshes/M_MED_Racer_Zero_Jacket_zip_Dyn
bodies {
 dyn_zipper_chest [{shapeType=Capsule, length=1, radius=0.75, shapeName=dyn_zipper_chest_capsule, center=<2 0 0>, rotation=<rot 90 0 0>}]
 dyn_zipper_chest_MSTR [{shapeType=Capsule, length=0.505, radius=0.505}]
 pelvis [{shapeType=Capsule, length=8.7745, radius=10.8041, center=<-2.4193 -0.4175 0>}]
 spine_05 [{shapeType=Capsule, length=7.1972, radius=12.3814, center=<-1.1184 -1.0415 0>}]
}
constraints [
 {
  bone1Name=dyn_zipper_chest_MSTR
  bone2Name=spine_05
  swing1LimitDegrees=45
  swing1Motion=Free
  swing2LimitDegrees=45
  swing2Motion=Free
  twistLimitDegrees=45
  twistMotion=Free
 }
 {bone1Name=dyn_zipper_chest, bone2Name=dyn_zipper_chest_MSTR, swing1LimitDegrees=30, swing2LimitDegrees=30, twistLimitDegrees=5}
 {
  bone1Name=spine_05
  bone2Name=pelvis
  swing1LimitDegrees=45
  swing1Motion=Free
  swing2LimitDegrees=45
  swing2Motion=Free
  twistLimitDegrees=45
  twistMotion=Free
 }
]
